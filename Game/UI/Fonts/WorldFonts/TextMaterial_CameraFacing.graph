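# Unlit, two-sided masked material for world-space text that turns to face the camera.
# AlignMeshToTheCamera supplies the camera-facing offset, using three vector parameters as a custom object basis to set text orientation.
# The red channel of a Font parameter, defaulting to Roboto distance field, drives a steep lerp that sharpens glyph edges for masking.
# Vertex color times an EmissiveMult scalar provides tintable brightness.
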Material /Game/UI/Fonts/WorldFonts/TextMaterial_CameraFacing
graph {
 properties {blendMode=BLEND_Masked, pixelDepthOffsetMode=PDOM_Legacy, shadingModel=MSM_Unlit, twoSided=true, usedShadingModels=Unlit}
 nodes [
  LinearInterpolate(Alpha=FontSampleParameter() {font=/Engine/EngineFonts/RobotoDistanceField, parameterName=Font}.R) {constA=-90, constB=90}
  MaterialFunctionCall!(
   WorldPosition
   PivotLocation
   WS Normals
   Custom Object Basis 1=VectorParameter() {parameterName=Basis1, defaultValue=<rgba 0 0 1 1>}.RGB
   Custom Object Basis 2=VectorParameter() {parameterName=Basis2, defaultValue=<rgba 1 0 0 1>}.RGB
   Custom Object Basis 3=VectorParameter() {parameterName=Basis3, defaultValue=<rgba 0 -1 0 1>}.RGB
  ) {materialFunction=/Engine/Functions/Engine_MaterialFunctions02/WorldPositionOffset/AlignMeshToTheCamera}
  Multiply(A=VertexColor(), B=ScalarParameter() {defaultValue=1, parameterName=EmissiveMult})
 ]
}
